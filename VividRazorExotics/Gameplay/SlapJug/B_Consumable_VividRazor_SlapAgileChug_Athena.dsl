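; Consumable Blueprint for the VividRazor Slap agile chug item, handling color variant setup for its animation-driven effects.
; When an anim notify in a montage spawns VFX, it writes the FX variant index into that component's ColorVariant integer parameter.
; BeginPlay, Tick and the construction script add no logic of their own, so any consumption behavior lies outside this graph.
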(event EventBeginPlay
  (|CallParentFunction) ; This node is disabled and will not be called.\nDrag off pins to build functionality.
)

(event EventTick
  (params (DeltaSeconds :type Float%20%28single-precision%29))
  (|CallParentFunction DeltaSeconds) ; This node is disabled and will not be called.\nDrag off pins to build functionality.
)

(event EventAnimNotify_VFX_VisualSetup
  (params (Spawn%20Notify%20VFX :type FXSystem%20Component%20Object%20Reference))
  ;[1 1 1] Set the Color Variant Info on FXs spawned from AnimNotify Events in Animation Montages
  (Effects|Components|ParticleSystem|SetIntParameter Spawn%20Notify%20VFX "ColorVariant" (Variables|Default|GetFX%20Variant%20Index)))

(fn ConstructionScript
  (|CallParentFunction))
