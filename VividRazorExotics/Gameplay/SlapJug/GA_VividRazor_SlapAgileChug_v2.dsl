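; Gameplay ability for the Slap jug consumable: a channeled drink that restores health and shield and applies a Slap effect.
; A montage trigger starts a repeating server-side tick that commits ammo and applies SetByCaller healing, with hotfix-driven health-first and overheal rules.
; Slap applies on activation and optionally on end.
; Activation fails at full health and shield; input release can end channeling, while empty ammo plays an outro that removes the item.
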(event Ability|EventActivateAbility
  ;[0 1 0] Activate ability and wait for trigger section
  (Utilities|FlowControl|Sequence
    (:then_0
      (CallFunction|CheckforInfiniteAmmo)
      (Ability|Tasks|PlayMontageandWaitforTrigger "None" (Variables|Montage|GetInUseMontage)
        (:then
          (EventDispatchers|BindEventtoOnLocalAmmoChanged (Ability|GetCurrentSourceWeapon) (EventDispatchers|CreateEvent :!event Custom|OnWeaponAmmoChanged))
          (CallFunction|ShowTimerUntilApplication)
          (bind _returnvalue (Ability|ApplyGameplayEffectToOwner (Variables|GameplayEffect|GetGE_Channeling)))
          (Variables|GameplayEffect|SetHandleChannelingGE _returnvalue)
          (Ability|Tasks|WaitInputRelease true
            (:then
              ; Check slap application
              (Variables|Default|SetApplySlapOnStartAbility (> (ScalableFloat|GetValueatLevel (Variables|Default|GetApplySlapOnStartAbilityHF)) 0))
              (Variables|Default|SetApplySlapOnEndAbility (> (ScalableFloat|GetValueatLevel (Variables|Default|GetApplySlapOnEndAbilityHF)) 0)))
            (:OnRelease
              (CallFunction|HandleRelease))))
        (:Triggered
          (Variables|Default|SetStartedChanelling true)
          (CallFunction|AdjustValuesforEffectiveHealthFirst)
          (bind _returnvalue_2 (Utilities|Time|SetTimerbyEvent (EventDispatchers|CreateEvent :!event Custom|ApplyEffectIncremental) (ScalableFloat|GetValueatLevel (Variables|Hotfix|GetRow_ApplicationRate)) true))
          (Variables|Default|SetApplicationTimer _returnvalue_2)
          (CallFunction|ApplyEffectIncremental)
          (CallFunction|EnableRelease)
          (if (Variables|Default|GetApplySlapOnStartAbility)
            (Utilities|FlowControl|DoOnce false
              (:Completed
                (CallFunction|ApplySlapEffect)))))
        (:Cancelled
          _)
        (:Completed
          _))
      (CallFunction|PreparetoEndAbility))
    (:then_1
      (CallFunction|ResetSlapatStart))))

(event Custom|ResetSlapAtStart
  ;[0 1 0] Activate ability and wait for trigger section
  (Utilities|FlowControl|DoOnce false
    (:Completed
      (CallFunction|ApplySlapEffect))))

(event Custom|ApplyEffectIncremental
  (bind (_healthfull _shieldfull) (CallFunction|IsPawnFullHealthandShield (CallFunction|GetAvatarasFortPlayerPawn)))
  ;[0 0.5221 1] Apply Healing to Activating Pawn
  (if (Networking|IsServer)
    (if (Math|Boolean|NANDBoolean _healthfull _shieldfull)
      (bind _returnvalue (Ability|CommitAbility))
      (if _returnvalue
        (bind (_healamountout _shieldamountout) (CallFunction|CalculateHealthandShield))
        (bind _gameplayeffect (CallFunction|SetupGE _healamountout _shieldamountout))
        (bind _returnvalue_2 (Ability|ApplyGameplayEffectSpecToOwner _gameplayeffect))
        (Variables|GameplayEffect|SetHandleAppliedGE _returnvalue_2)
        (CallFunction|ApplyHealthandShieldVisuals _healamountout _shieldamountout)
        (Variables|Default|SetAppliedHealing true)
        (else
          _)))
    (Ability|EndAbility)))

(event Custom|ShowTimerUntilApplication
  ;[1 0 0.75] Timer Amount
  (if (ScalableFloat|GetValueasBool (Variables|Hotfix|GetRow_ShowTimerUntilApplication))
    (bind _inusemontage (Variables|Montage|GetInUseMontage)
      (:then
        (bind _returnvalue (ContextualAnim|Utilities|GetSectionLength _inusemontage))
        (Ability|AddGameplayCueWithParamsToOwner (Variables|Default|GetGC_EquippingCue) (Ability|GameplayCue|MakeGameplayCueParameters :NormalizedMagnitude 0.0 :RawMagnitude _returnvalue :EffectContext "()" :AggregatedSourceTags "(GameplayTags=)" :AggregatedTargetTags "(GameplayTags=)" :Location "0,0,0" :Normal "0,0,0" :GameplayEffectLevel 1 :AbilityLevel 1 :bReplicateLocationWhenUsingMinimalRepProxy false)))
      (:else))))

(event Custom|OnWeaponAmmoChanged
  (params (LocalCount :type Integer) (LocalRemaining :type Integer))
  ;[0.03429 0.3069 0.7899] Handle ammo change + input release
  (if (<= LocalRemaining 0)
    (Ability|Tasks|SetNextMontageSectionandWait (Animation|GetCurrentMontage) (Variables|Montage|GetInUseMontageEndSection)
      (:then
        (Variables|Montage|SetPlayingOutro? true))
      (:OnComplete
        (CallFunction|PreparetoEndAbility)))))

(event Custom|HandleRelease
  ;[0.03429 0.3069 0.7899] Handle ammo change + input release
  (if (ScalableFloat|GetValueasBool (Variables|Hotfix|GetRow_HoldInputToUse))
    (Ability|HasAuthority
      (:False
        (if (ScalableFloat|GetValueasBool (Variables|Hotfix|GetRow_IgnoreCancelUntilFirstApplication))
          (if (not (Variables|Default|GetStartedChanelling))
            (Utilities|FlowControl|WaitforBothExecutions true
              (:Then
                _)))
          (if (not (Variables|Montage|GetPlayingOutro?))
            (CallFunction|PreparetoEndAbility))
          (else
            (CallFunction|PreparetoEndAbility)))))))

(event Custom|EnableRelease
  ;[0.03429 0.3069 0.7899] Handle ammo change + input release
  (Utilities|FlowControl|WaitforBothExecutions true
    (:Then
      (if (not (Variables|Montage|GetPlayingOutro?))
        (CallFunction|PreparetoEndAbility)))))

(event Custom|PrepareToEndAbility
  ;[1 0.5 0] Prepare to end the ability
  (Utilities|FlowControl|Sequence
    (:then_0
      (Ability|HasAuthority
        (:True
          (if (Variables|Default|GetAppliedHealing)
            (GameplayCue|ExecuteGameplayCueOnActor(Burst) (Ability|GetAvatarActorfromActorInfo) (Variables|Default|GetHealFinishedGameplayCue) (Ability|GameplayCue|MakeGameplayCueParameters 0.0))))))
    (:then_1
      (Utilities|FlowControl|DelayUntilNextTick)
      (Ability|EndAbility))))

(event Ability|EventOnEndAbility
  (params (bWasCancelled :type Boolean))
  ;[1 0.1038 0.05208] End Ability
  (Ability|RemoveGameplayEffectFromOwnerWithHandle (Variables|GameplayEffect|GetHandleChannelingGE))
  (if (Variables|Default|GetStartedChanelling)
    (Variables|Default|SetStartedChanelling false)
    (if (Variables|Default|GetApplySlapOnEndAbility)
      (CallFunction|ApplySlapEffect))
    (Variables|Default|SetAppliedHealing false)
    (Utilities|Time|ClearandInvalidateTimerbyHandle (Variables|Default|GetApplicationTimer))
    (if (Variables|Montage|GetPlayingOutro?)
      (CallFunction|RemoveItem) ; ???
      (Ability|SendGameplayEvent (Variables|Default|GetEvent_HealConsumed))
      (else
        (Ability|Animation|MontageStop -1.0)))))

(event Custom|ApplySlapEffect
  ;[1 0 0 0.5] Apply Slap
  (Ability|ApplyGameplayEffectSpecToOwner (Ability|MakeOutgoingGameplayEffectSpec (Variables|Default|GetSlapGE))))

(fn AdjustValuesForEffectiveHealthFirst
  (if (ScalableFloat|GetValueasBool (Variables|Hotfix|GetRow_HealHealthFirstHF))
    (Default|SetAmmoCost 2.0)
    (Utilities|Struct|SetmembersinScalableFloat (Variables|Hotfix|GetRow_HealAmount) 2.0) ; Multiply the cost by 0 for Infinite Ammo, otherwise multiply by 1.
    (return)
    (else
      (return))))

(fn ApplyHealthAndShieldVisuals
  (params (HealAmount :type Float%20%28double-precision%29) (ShieldAmount :type Float%20%28double-precision%29))
  (bind _gc_healthheal (Variables|GameplayCue|GetGC_HealthHeal))
  (bind _gc_shieldheal (Variables|GameplayCue|GetGC_ShieldHeal))
  (Utilities|FlowControl|Sequence
    (:then_0
      ; Health Heal cue
      (if (and (> (Variables|GetHealAmount) 0) (GameplayTags|IsGameplayTagValid _gc_healthheal))
        (Ability|ExecuteGameplayCueOnOwner _gc_healthheal)))
    (:then_1
      ; Shield Heal cue
      (if (and (> (Variables|GetShieldAmount) 0) (GameplayTags|IsGameplayTagValid _gc_shieldheal))
        (Ability|ExecuteGameplayCueOnOwner _gc_shieldheal)))))

(fn CalculateHealthAndShield
  (bind _row_healamount (Variables|Hotfix|GetRow_HealAmount))
  (bind _returnvalue (ScalableFloat|GetValueatLevel _row_healamount))
  (bind _row_healamount_2 (Variables|Hotfix|GetRow_HealAmount))
  (bind _returnvalue_2 (ScalableFloat|GetValueatLevel _row_healamount_2))
  (bind _tofullhealthamount (LocalVariables|GetToFullHealthAmount))
  (bind _tofullhealthamount_2 (LocalVariables|GetToFullHealthAmount))
  (bind _outfortplayerpawn (CallFunction|GetAvatarasValidFortPlayerPawn)
    (:ValidFortPlayerPawn
      (LocalVariables|SetFortPlayerPawn _outfortplayerpawn)
      (Utilities|FlowControl|Sequence
        (:then_1
          (if (ScalableFloat|GetValueasBool (Variables|Hotfix|GetRow_HealHealthFirstHF))
            ; Calculate heal and over heal
            (LocalVariables|SetToFullHealthAmount (- (FortPawn|GetMaxHealth (LocalVariables|GetFortPlayerPawn)) (FortPawn|GetHealth (LocalVariables|GetFortPlayerPawn))))
            (LocalVariables|SetHealAmount (Math|Float|Min(Float) _returnvalue_2 _tofullhealthamount))
            (LocalVariables|SetOverHeal (Math|Float|Clamp(Float) (- _returnvalue_2 _tofullhealthamount) 0.0 _returnvalue_2))
            (if (and (< (LocalVariables|GetToFullHealthAmount) 1e+01) (> (LocalVariables|GetToFullHealthAmount) 0))
              (LocalVariables|SetApplyOverHeal? true)
              (elif (>= (FortPawn|GetHealth (LocalVariables|GetFortPlayerPawn)) (FortPawn|GetMaxHealth (LocalVariables|GetFortPlayerPawn)))
                ; Check if White health is full before healing shields
                (LocalVariables|SetApplyOverHeal? false)))
            ; Calculate Shields
            (bind _output_get (LocalVariables|SetToFullShieldAmount (- (FortPawn|GetMaxShield (LocalVariables|GetFortPlayerPawn)) (FortPawn|GetShield (LocalVariables|GetFortPlayerPawn)))))
            (LocalVariables|SetShieldAmount (Math|Float|Clamp(Float) _output_get 0.0 (Math|Float|SelectFloat (LocalVariables|GetOverHeal) (ScalableFloat|GetValueatLevel (Variables|Hotfix|GetRow_HealAmount)) (LocalVariables|GetApplyOverHeal?))))
            (else
              (Utilities|FlowControl|Sequence
                (:then_0
                  ; Calculate heal and over heal
                  (LocalVariables|SetToFullHealthAmount (- (FortPawn|GetMaxHealth (LocalVariables|GetFortPlayerPawn)) (FortPawn|GetHealth (LocalVariables|GetFortPlayerPawn))))
                  (LocalVariables|SetHealAmount (Math|Float|Min(Float) _returnvalue _tofullhealthamount_2))
                  (LocalVariables|SetOverHeal (Math|Float|Clamp(Float) (- _returnvalue _tofullhealthamount_2) 0.0 _returnvalue)))
                (:then_1
                  ; Calculate Shields
                  (bind _output_get_2 (LocalVariables|SetToFullShieldAmount (- (FortPawn|GetMaxShield (LocalVariables|GetFortPlayerPawn)) (FortPawn|GetShield (LocalVariables|GetFortPlayerPawn)))))
                  (LocalVariables|SetShieldAmount (Math|Float|Clamp(Float) _output_get_2 0.0 (Math|Float|SelectFloat (+ (LocalVariables|GetOverHeal) (ScalableFloat|GetValueatLevel (Variables|Hotfix|GetRow_HealAmount))) (ScalableFloat|GetValueatLevel (Variables|Hotfix|GetRow_HealAmount)) (ScalableFloat|GetValueasBool (Variables|Hotfix|GetRow_OverHealAppliedToShields))))))))))
        (:then_2
          (return (LocalVariables|GetHealAmount) (LocalVariables|GetShieldAmount)))))
    (:AvatarCastFailed
      (return 0.0 0.0))))

(fn CanApplyHolster
  (params (HolsterId :type Name%20%28by%20ref%29 :default "None"))
  (return false))

(fn CheckForInfiniteAmmo
  (bind _returnvalue (Ability|GetCurrentSourceWeapon))
  ;[0.4536 1 0.3115 0.5]
  (Utilities|IsValid _returnvalue
    (:Is%20Valid
      (bind _output_get (LocalVariables|SetbInfiniteAmmo (or (FortWeapon|Ammo|HasInfiniteUses _returnvalue) (FortWeapon|Ammo|InventoryOwnerHasInfiniteUses _returnvalue))))
      (if _output_get
        (FortItem|BroadcastonItemChanged (FortWeapon|GetInventoryItem _returnvalue) false true false false) ; Trigger the ammo count to change to/from an infinity symbol, which doesn't otherwise trigger because the cost is 0.
      ))
    (:Is%20Not%20Valid
      _))
  (Default|SetAmmoCost (Math|Conversions|ToFloat(Boolean) (not (LocalVariables|GetbInfiniteAmmo)))) ; Multiply the cost by 0 for Infinite Ammo, otherwise multiply by 1.
)

(fn GetCustomAbilitySourceTransform
  (bind (_returnvalue_x _returnvalue_y _returnvalue_z) (Math|Vector|BreakVector (Transformation|GetSocketLocation (Variables|/Script/Engine.Character|GetMesh (CallFunction|GetAvatarasFortPlayerPawn)) "pelvis")))
  (return (Math|Vector|MakeVector (.x (Transformation|GetActorLocation (CallFunction|GetAvatarasFortPlayerPawn))) (.y (Transformation|GetActorLocation (CallFunction|GetAvatarasFortPlayerPawn))) _returnvalue_z) "0, 0, 0" "1.000000,1.000000,1.000000"))

(fn IsPawnFullHealthAndShield
  (params (FortPlayerPawn :type Fort%20Player%20Pawn%20Object%20Reference))
  (bind _output_get (LocalVariables|SetHealthFullOut (Math|Float|NearlyEqual(Float) (FortPawn|GetHealth FortPlayerPawn) (FortPawn|GetMaxHealth FortPlayerPawn) 0.5)))
  (bind _output_get_2 (LocalVariables|SetShieldFullOut (Math|Float|NearlyEqual(Float) (FortPawn|GetShield FortPlayerPawn) (FortPawn|GetMaxShield FortPlayerPawn) 0.5)))
  (return _output_get _output_get_2))

(fn CanActivateAbility
  (params (ActorInfo :type Gameplay%20Ability%20Actor%20Info%20Structure) (Handle :type Gameplay%20Ability%20Spec%20Handle%20Structure))
  (bind (_owneractor _avataractor) (Utilities|Struct|BreakGameplayAbilityActorInfo ActorInfo))
  (bind _asfort_player_pawn (Utilities|Casting|CastToFortPlayerPawn _avataractor)
    (:then
      (if (not (and (CallFunction|IsPawnFullHealthandShield _asfort_player_pawn) (.ShieldFull (CallFunction|IsPawnFullHealthandShield _asfort_player_pawn))))
        (return 0 true)))
    (:CastFailed
      _))
  (return (Variables|Default|GetTags_FailedResponse) false))

(fn RemoveItem
  ; not sure what this is doing
  (Inventory|RemoveFortItemFromPlayer (Pawn|GetController (Ability|GetActivatingPawn)) (FortWeapon|GetInventoryItem (Ability|GetCurrentSourceWeapon)) 1 true))

(fn SetupGE
  (params (HealAmount :type Float%20%28double-precision%29) (ShieldAmount :type Float%20%28double-precision%29))
  (bind _returnvalue (Ability|GameplayEffect|AssignTagSetbyCallerMagnitude (Ability|MakeOutgoingGameplayEffectSpec (Variables|GameplayEffect|GetGE_EffectToApply)) "(TagName=\"SetByCaller.Healing\")" (Variables|GetHealAmount)))
  (bind _returnvalue_2 (Ability|GameplayEffect|AssignTagSetbyCallerMagnitude _returnvalue "(TagName=\"SetByCaller.Shield\")" (Variables|GetShieldAmount)))
  (return _returnvalue_2))
